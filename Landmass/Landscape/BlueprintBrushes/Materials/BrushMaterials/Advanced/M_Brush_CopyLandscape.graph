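# Landmass brush material that applies a copied landscape texture to a landscape as height or weightmap, chosen by a static switch.
# Height is unpacked from RG8, scaled, offset, blended with the previous height through the shape falloff mask and repacked.
# Weightmap output takes the copy's red channel and weightmap blend modes, with noise and opacity available.
# UV logic handles indexed subdivision tiles with shared or separate edges, plus wrapping within render-target padding.
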Material /Landmass/Landscape/BlueprintBrushes/Materials/BrushMaterials/Advanced/M_Brush_CopyLandscape
graph {
 properties {floatPrecisionMode=MFPM_Full, pixelDepthOffsetMode=PDOM_Legacy, shadingModel=MSM_Unlit, usedShadingModels=Unlit}
 nodes [
  n1 = TextureSampleParameter2D(UVs=NamedRerouteUsage()) {parameterName=CopiedTexture, texture=/Engine/ArtTools/RenderToTexture/Textures/T_EV_BlankWhite_01}
  NamedRerouteDeclaration(None=@n1.R) {name=CopiedWeightmap, nodeColor=<rgba 0.9219 0 1 1>}
  n2 = MaterialFunctionCall!(RG=ComponentMask(None=@n1.RGB) {g=true, r=true}) {materialFunction=/Landmass/Landscape/OldPrototype_BP/Mats/MF/UnpackRG8Height}
  n3 = ScalarParameter() {defaultValue=1, parameterName=ZScale}
  n4 = ScalarParameter() {parameterName=HeightOffset}
  n5 = ScalarParameter() {parameterName=BrushZ}
  NamedRerouteDeclaration(None=Add(A=Multiply(A=@n2.Result, B=@n3), B=Add(A=@n4, B=@n5))) {name=Current Height, nodeColor=<rgba 0.1484 1 0 1>}
  NamedRerouteDeclaration(None=NamedRerouteUsage()) {name=Weightmap Alpha, nodeColor=<rgba 0.078 0.5071 1 1>}
  n6 = MaterialFunctionCall!(Brush Alpha=NamedRerouteUsage()) {
   desc="Struct: Landmass_NoisePass_WeightMapSettings\r\nMacro: "
   materialFunction=/Landmass/Landscape/BlueprintBrushes/Materials/BrushMaterials/BasicFunctions/M_Brush_WeightmapNoise
  }
  NamedRerouteDeclaration(None=Multiply(A=@n6.Weightmap Noise, B=ScalarParameter() {defaultValue=1, parameterName=WeightmapOpacity})) {name=Weightmap Alpha with Noise, nodeColor=<rgba 0.5 1 0 1>}
  n7 = NamedRerouteUsage()
  n8 = ScalarParameter() {parameterName=Index}
  n9 = ScalarParameter() {defaultValue=1, parameterName=Subdivisions}
  n10 = Add(
   A=Multiply(A=NamedRerouteUsage(), B=Subtract(A=@n7)) {constB=1023}
   B=Multiply(
    A=MaterialFunctionCall!(1D Index=@n8, Cells XY=@n9) {materialFunction=/Engine/Functions/Engine_MaterialFunctions01/Coordinates/1Dto2DIndex}.2D Index
    B=NamedRerouteUsage()
   )
  )
  MaterialExpressionTextureObjectParameter_0 = TextureObjectParameter() {parameterName=CopiedTexture}
  n11 = Custom!(Tex=@MaterialExpressionTextureObjectParameter_0) {
   code="uint3 res;\r\nTex.GetDimensions(0,res.x,res.y,res.z);\r\nreturn float2(res.xy);"
   description=GetTextureResolution
   outputType=CMOT_Float2
   inputs [{
     inputName=Tex
     input {expression=&self:MaterialExpressionTextureObjectParameter_0}
    }]
  }
  NamedRerouteDeclaration(None=Divide(A=Add(A=@n10) {constB=0.5}, B=@n11)) {name=Non Shared Edges, nodeColor=<rgba 0 0.4375 1 1>}
  n12 = MaterialFunctionCall!(Falloff Width Override, Falloff Offset Override) {materialFunction=/Landmass/Landscape/BlueprintBrushes/Materials/BrushMaterials/BasicFunctions/LandmassBrush_ShapeMask}
  NamedRerouteDeclaration(None=@n12.Brush Edge Falloff) {name=Brush Edge Mask, nodeColor=<rgba 1 0 0.1641 1>}
  n13 = Divide(A=NamedRerouteUsage(), B=ScalarParameter() {defaultValue=1, parameterName=Subdivisions})
  n14 = ScalarParameter() {parameterName=Index}
  n15 = ScalarParameter() {defaultValue=1, parameterName=Subdivisions}
  n16 = CollectionParameter() {collection=/Landmass/Landscape/BlueprintBrushes/MPC/MPC_Landscape, parameterName=WorldSizeX}
  n17 = CollectionParameter() {collection=&MPC_Landscape, parameterName=WorldSizeY}
  n18 = CollectionParameter() {collection=&MPC_Landscape, parameterName=RTWorldSize}
  n19 = Multiply(
   A=Divide(
    A=MaterialFunctionCall(1D Index=@n14, Cells XY=@n15) {materialFunction=&1Dto2DIndex}.2D Index
    B=@n15
   )
   B=Divide(A=AppendVector(A=@n16, B=@n17), B=ComponentMask(None=@n18) {g=true, r=true})
  )
  NamedRerouteDeclaration(None=LinearInterpolate(A=NamedRerouteUsage(), B=Add(A=@n13, B=@n19), Alpha=ScalarParameter() {defaultValue=1, parameterName=Use Shared Edges})) {name=Subdivision UVs, nodeColor=<rgba 0.5703 1 0 1>}
  n20 = Multiply(A=Subtract(A=NamedRerouteUsage(), B=NamedRerouteUsage())) {constB=0.5, desc=Choose midpoint of the unused padding}
  n21 = NamedRerouteDeclaration(None=Divide(A=Max(A=@n20) {desc=at least one pixel}, B=NamedRerouteUsage()) {desc=convert to UVs}) {name=TexcoordWrapAmount, nodeColor=<rgba 0.9219 0 1 1>}
  n22 = Add(A=TextureCoordinate(), B=@n21)
  NamedRerouteDeclaration(None=Subtract(A=Frac(None=@n22), B=@n21)) {name=Wrapped Edge UVs, nodeColor=<rgba 0.5703 0 1 1>}
  n23 = MaterialFunctionCall() {materialFunction=/Landmass/Landscape/BlueprintBrushes/MF/LandmassBrush_UVs}
  NamedRerouteDeclaration(None=If(A=ScalarParameter() {defaultValue=1, parameterName=Subdivisions}, "A > B"=NamedRerouteUsage(), "A == B"=@n23.UVs, "A < B"=@n23.UVs) {constB=1}) {name=Final Brush UVs, nodeColor=<rgba 0.3359 1 0 1>}
  n24 = AppendVector(
   A=CollectionParameter() {collection=&MPC_Landscape, parameterName=LSQuadsX}
   B=CollectionParameter() {collection=&MPC_Landscape, parameterName=LSQuadsY}
  )
  NamedRerouteDeclaration(None=Add(A=NamedRerouteDeclaration(None=@n24) {name=Landscape Quads, nodeColor=<rgba 1 0 0.1172 1>})) {name=Landscape Verts, nodeColor=<rgba 0.8281 0 1 1>}
  n25 = NamedRerouteUsage()
  n26 = NamedRerouteUsage()
  NamedRerouteDeclaration(None=Saturate(None=MaterialFunctionCall!(Weightmap Mask=@n25, Mask=@n26) {materialFunction=/Landmass/Landscape/BlueprintBrushes/MF/BrushBlendModes_Weight}.Result)) {name=Weightmap Result, nodeColor=<rgba 0.5234 0 1 1>}
  NamedRerouteDeclaration(None=AppendVector(
    A=CollectionParameter() {collection=&MPC_Landscape, parameterName=RTResX}
    B=CollectionParameter() {collection=&MPC_Landscape, parameterName=RTResY}
   )) {name=RT Res, nodeColor=<rgba 0 1 0.5078 1>}
  NamedRerouteDeclaration(None=@n12.Previous Height) {name=Previous Height, nodeColor=<rgba 0 0.4609 1 1>}
  n27 = MaterialFunctionCall!(Current Height=NamedRerouteUsage(), Edge Mask=NamedRerouteUsage(), Displacement Mask=Constant() {r=1}) {materialFunction=/Landmass/Landscape/BlueprintBrushes/MF/BrushBlendModes_Height}
  n28 = ScalarParameter() {parameterName=DebugOffset}
  StaticSwitchParameter(
   True=NamedRerouteUsage()
   False=MaterialFunctionCall!(Height Int -32k to 32k=Add(A=@n27.Result, B=@n28)) {materialFunction=/Landmass/Landscape/BlueprintBrushes/MF/PackRG8Height}.XY
  ) {parameterName=Weightmap}
 ]
}
