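# Unlit Landmass brush material that stamps a big island into landscape height or weightmap data, selected by a Weightmap static switch.
# A shape mask supplies edge falloff and base height, with noise-driven parameters for shoreline cliff variation and edge breakup.
# An underwater mask and ocean depth feed height blending alongside a displacement texture, then height is packed to RG8.
# The weightmap branch has its own falloff, noise and blend function.
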Material /Landmass/Landscape/BlueprintBrushes/Materials/BrushMaterials/Advanced/M_Brush_TextureStamp_BigIsland
graph {
 properties {floatPrecisionMode=MFPM_Full, pixelDepthOffsetMode=PDOM_Legacy, shadingModel=MSM_Unlit, usedShadingModels=Unlit}
 nodes [
  n1 = ScalarParameter() {defaultValue=0.1, parameterName=FalloffWidth}
  Abs(None=@n1)
  n2 = Constant()
  n3 = MaterialFunctionCall() {materialFunction=/Landmass/Landscape/BlueprintBrushes/MF/LandmassBrush_UVs}
  n4 = Abs(None=Subtract(A=NamedRerouteDeclaration(None=@n3.UVs) {name=UVs, nodeColor=<rgba 0.3542 1 0 1>}) {constB=0.5})
  n5 = Subtract(B=NamedRerouteUsage()) {constA=0.5}
  n6 = Multiply(A=NamedRerouteUsage()) {constB=2}
  n7 = Subtract(B=Add(A=Multiply(A=@n5, B=@n6), B=NamedRerouteUsage())) {constA=0.5}
  n8 = Max(A=Add(A=NamedRerouteUsage(), B=Subtract(A=@n4, B=@n7))) {constB=0}
  n9 = Max(A=Add(A=Distance(A=@n2, B=@n8), B=@n7)) {constB=0}
  n10 = NamedRerouteUsage()
  n11 = Subtract(B=@n10) {constA=0.5}
  n12 = OneMinus(None=Divide(A=Subtract(A=@n9, B=@n11), B=@n10))
  Abs(None=@n12)
  MaterialFunctionCall(In=Saturate(None=Multiply(A=@n12) {constB=-1}), Power=ScalarParameter() {defaultValue=1, parameterName=FalloffPower}) {materialFunction=&SCurve}
  MaterialFunctionCall(In=Saturate(None=@n12), Power=ScalarParameter() {defaultValue=1, parameterName=FalloffPower}) {materialFunction=&SCurve}
  n13 = LinearInterpolate(B=NamedRerouteUsage(), Alpha=ScalarParameter() {defaultValue=1, parameterName=RelativeDepth}) {constA=1}
  n14 = ScalarParameter() {parameterName=FalloffOffset}
  n15 = Saturate(None=Divide(A=Subtract(A=@n13, B=@n14), B=@n1))
  n16 = NamedRerouteUsage()
  Multiply(A=@n15, B=@n16)
  n17 = Multiply(
   A=MaterialFunctionCall() {materialFunction=&LandmassBrush_UVs}.UVs
   B=ScalarParameter() {defaultValue=16, parameterName=ShoreVariationTiling}
  )
  n18 = ScalarParameter() {defaultValue=1, parameterName=ShoreVariationSeed}
  NamedRerouteDeclaration(None=MaterialFunctionCall!(
    Position=AppendVector(A=@n17, B=@n18)
    Octaves=ScalarParameter() {defaultValue=8, parameterName=ShoreVariationOctaves}
    Persistence=ScalarParameter() {defaultValue=0.35, parameterName=ShoreVariationPersistence}
    Turbulence
   ) {materialFunction=/Landmass/Landscape/BlueprintBrushes/MF/LandmassBrush_Noise}.Result) {name=ShoreVariationNoise, nodeColor=<rgba 0 0.4141 1 1>}
  n19 = MaterialFunctionCall!(Falloff Width Override, Falloff Offset Override=NamedRerouteUsage()) {materialFunction=/Landmass/Landscape/BlueprintBrushes/Materials/BrushMaterials/BasicFunctions/LandmassBrush_ShapeMask}
  NamedRerouteDeclaration(None=@n19.Brush Base Height) {name=Current Height, nodeColor=<rgba 0.1484 1 0 1>}
  n20 = Divide(
   A=Subtract(A=OneMinus(None=@n16), B=ScalarParameter() {defaultValue=0.1, parameterName=ExtraOffset})
   B=ScalarParameter() {defaultValue=1, parameterName=FalloffAdjust}
  )
  NamedRerouteDeclaration(None=LinearInterpolate(A=@n15, B=Saturate(None=@n20), Alpha=ScalarParameter() {parameterName=EnableDefaultFalloff}) {constA=1}) {name=Weightmap Alpha, nodeColor=<rgba 0.078 0.5071 1 1>}
  n21 = MaterialFunctionCall!(Brush Alpha=NamedRerouteUsage()) {
   desc="Struct: Landmass_NoisePass_WeightMapSettings\r\nMacro: "
   materialFunction=/Landmass/Landscape/BlueprintBrushes/Materials/BrushMaterials/BasicFunctions/M_Brush_WeightmapNoise
  }
  NamedRerouteDeclaration(None=Multiply(A=@n21.Weightmap Noise, B=ScalarParameter() {defaultValue=1, parameterName=WeightmapOpacity})) {name=Weightmap Alpha with Noise, nodeColor=<rgba 0.5 1 0 1>}
  NamedRerouteDeclaration(None=ScalarParameter() {defaultValue=0.2, parameterName=FalloffRadius}) {name=Falloff Radius, nodeColor=<rgba 0.1016 1 0 1>}
  n22 = Power(Base=NamedRerouteUsage(), Exp=ScalarParameter() {defaultValue=4, group=Shoreline, parameterName=Shoreline Variation Power})
  n23 = ScalarParameter() {defaultValue=2, group=Shoreline, parameterName=Shoreline Variation Boost}
  n24 = LinearInterpolate(
   A=ScalarParameter() {defaultValue=1, group=Shoreline, parameterName=Shoreline Width Meters}
   B=ScalarParameter() {defaultValue=10, group=Shoreline, parameterName=Shoreline Cliff Multiplier}
   Alpha=Saturate(None=Multiply(A=@n22, B=@n23))
  )
  NamedRerouteDeclaration(None=@n24) {name=Shoreline Cliff Variation, nodeColor=<rgba 0.4062 1 0 1>}
  n25 = Multiply(A=@n19.Brush Edge Falloff Linear) {constB=-1}
  NamedRerouteDeclaration(None=MaterialFunctionCall(In=Saturate(None=@n25), Power=ScalarParameter() {defaultValue=1, parameterName=FalloffPower}) {materialFunction=&SCurve}.Result) {name=Underwater Mask, nodeColor=<rgba 0.2422 0 1 1>}
  n26 = AppendVector(
   A=Multiply(
    A=MaterialFunctionCall() {materialFunction=&LandmassBrush_UVs}.UVs
    B=ScalarParameter() {defaultValue=2, parameterName=EdgeNoiseTiling}
   )
   B=ScalarParameter() {defaultValue=1, parameterName=EdgeNoiseSeed}
  )
  n27 = ScalarParameter() {defaultValue=4, parameterName=EdgeNoiseOctaves}
  n28 = ScalarParameter() {defaultValue=0.35, parameterName=EdgeNoisePersistence}
  NamedRerouteDeclaration(None=Multiply(
    A=MaterialFunctionCall(Position=@n26, Octaves=@n27, Persistence=@n28) {materialFunction=&LandmassBrush_Noise}.Result
    B=ScalarParameter() {defaultValue=0.2, parameterName=EdgeNoiseStrength}
   )) {name=Edge Falloff Noise, nodeColor=<rgba 0.0786 0.3441 1 1>}
  n29 = Multiply(A=@n19.Brush Edge Falloff, B=NamedRerouteUsage())
  NamedRerouteDeclaration(None=MaterialFunctionCall(In=Saturate(None=@n29), Power=ScalarParameter() {defaultValue=1, parameterName=FalloffPower}) {materialFunction=&SCurve}.Result) {name=Brush Edge Mask, nodeColor=<rgba 1 0 0.1641 1>}
  NamedRerouteDeclaration(None=Add(A=ScalarParameter() {parameterName=FalloffOffset}, B=NamedRerouteUsage())) {name=Offset, nodeColor=<rgba 1 0.2344 0 1>}
  n30 = VectorParameter() {parameterName=DisplacementChannel, defaultValue=<rgba 1 0 0 0>}
  n31 = AppendVector(A=@n30.RGB, B=@n30.RGB)
  n32 = TextureSampleParameter2D(UVs=MaterialFunctionCall() {materialFunction=&LandmassBrush_UVs}.UVs) {parameterName=Displacement, texture=/Engine/ArtTools/RenderToTexture/Textures/T_EV_BlankWhite_01}
  n33 = ScalarParameter() {parameterName=DisplacementReferencePlane}
  NamedRerouteDeclaration(None=Add(A=DotProduct(A=@n31, B=@n32.RGBA), B=OneMinus(None=@n33))) {name=Displacement Mask, nodeColor=<rgba 0 0.8125 1 1>}
  n34 = VectorParameter() {parameterName=DisplacementChannel, defaultValue=<rgba 1 0 0 0>}
  n35 = MaterialFunctionCall() {materialFunction=&LandmassBrush_UVs}
  NamedRerouteDeclaration(None=DotProduct(
    A=AppendVector(A=@n34.RGB, B=@n34.RGB)
    B=TextureSampleParameter2D(UVs=@n35.UVs) {parameterName=WeightmapTexture, texture=&T_EV_BlankWhite_01}.RGBA
   )) {name=WeightmapMask, nodeColor=<rgba 1 0 0.5859 1>}
  n36 = NamedRerouteUsage()
  n37 = NamedRerouteUsage()
  NamedRerouteDeclaration(None=Saturate(None=MaterialFunctionCall!(Weightmap Mask=@n36, Mask=@n37) {materialFunction=/Landmass/Landscape/BlueprintBrushes/MF/BrushBlendModes_Weight}.Result)) {name=Weightmap Result, nodeColor=<rgba 0.5234 0 1 1>}
  NamedRerouteDeclaration(None=Multiply(A=@n24, B=ScalarParameter() {defaultValue=0.1, parameterName=FalloffWidth}) {constB=100}) {name=Falloff Width, nodeColor=<rgba 0.3359 0 1 1>}
  NamedRerouteDeclaration(None=@n19.Previous Height) {name=Previous Height, nodeColor=<rgba 0 0.4609 1 1>}
  NamedRerouteUsage()
  ScalarParameter() {parameterName=Edge Offset}
  n38 = NamedRerouteUsage()
  n39 = NamedRerouteUsage()
  n40 = NamedRerouteUsage()
  n41 = Multiply(A=ScalarParameter() {defaultValue=-20, parameterName=Ocean Depth Meters}) {constB=100}
  n42 = MaterialFunctionCall!(
   Current Height=Add(A=LinearInterpolate(B=@n38, Alpha=@n39), B=Multiply(A=@n40, B=@n41))
   Edge Mask=Constant() {r=1}
   Displacement Mask=NamedRerouteUsage()
  ) {materialFunction=/Landmass/Landscape/BlueprintBrushes/MF/BrushBlendModes_Height}
  n43 = ScalarParameter() {parameterName=DebugOffset}
  StaticSwitchParameter(
   True=NamedRerouteUsage()
   False=MaterialFunctionCall!(Height Int -32k to 32k=Add(A=@n42.Result, B=@n43)) {materialFunction=/Landmass/Landscape/BlueprintBrushes/MF/PackRG8Height}.XY
  ) {parameterName=Weightmap}
 ]
}
